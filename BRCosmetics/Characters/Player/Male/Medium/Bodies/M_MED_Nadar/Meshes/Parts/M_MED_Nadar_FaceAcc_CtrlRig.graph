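# Control Rig for the Nadar medium male face accessory, which rotates the beak in response to head pose.
# A Spherical Pose Reader measures the head bone against spine_04 with an asymmetric active region and falloff, and has debug drawing enabled.
# Its output weights a Modify Transforms rotation on base_beak, the parent of dyn_beak, as a pose-driven corrective.
# The neck bend and targeting variables are declared but unused in the graph.
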ControlRigBlueprint /BRCosmetics/Characters/Player/Male/Medium/Bodies/M_MED_Nadar/Meshes/Parts/M_MED_Nadar_FaceAcc_CtrlRig
graphs [
 RigVMFunctionLibrary {links [], nodes []}
 RigVMModel {nodes [
   RigUnit_BeginExecution = &RigUnit_BeginExecution()
   RigVMFunction_Sequence = &RigVMFunction_Sequence(ExecuteContext=@RigUnit_BeginExecution.ExecutePin)
   n1 = &RigUnit_SphericalPoseReader(
    ActiveRegionScaleFactors=(PositiveWidth=3,NegativeWidth=0.1,PositiveHeight=6,NegativeHeight=6)
    ActiveRegionSize=0.1
    Debug=(bDrawDebug=True,DebugScale=25,DebugSegments=20,DebugThickness=0.25)
    DriverAxis=(X=1,Y=0,Z=0)
    DriverItem=(Type=Bone,Name="head")
    FalloffRegionScaleFactors=(PositiveWidth=-3,NegativeWidth=1,PositiveHeight=0,NegativeHeight=0)
    FalloffSize=0.25
    OptionalParentItem=(Type=Bone,Name="spine_04")
    RotationOffset=(X=90,Y=0,Z=-180)
    ExecutePin=@RigVMFunction_Sequence.A
   )
   > &RigUnit_ModifyTransforms(
    ItemToModify=((Item=(Type=Bone,Name="base_beak"),Transform=(Rotation=(X=0,Y=0,Z=0.5736,W=0.8192),Translation=(X=0,Y=0,Z=0),Scale3D=(X=1,Y=1,Z=1))))
    Weight=@n1.OutputParam
   )
  ]}
]
hierarchy [
 root {loc=<0 0 0>, rot=<0 0 0 1>}
 pelvis {parent=root, loc=<0 -0.7071 93.1191>, rot=<0.0127 0.707 -0.0127 -0.707>}
 spine_01 {parent=pelvis, loc=<6.7494 -0.6388 0>, rot=<0 0 0.034 -0.9994>}
 spine_02 {parent=spine_01, loc=<7.2518 -0.035 0>, rot=<0 0 0.0238 -0.9997>}
 spine_03 {parent=spine_02, loc=<6.0104 0 0>, rot=<0 0 0.0149 -0.9999>}
 spine_04 {parent=spine_03, loc=<8.1024 0 0>, rot=<0 0 -0.0132 -0.9999>}
 spine_05 {parent=spine_04, loc=<10.0146 0 0>, rot=<0 0 -0.0015 -1>}
 neck_01 {parent=spine_05, loc=<11.3138 0.1529 0>, rot=<0 0 -0.2003 -0.9797>}
 neck_02 {parent=neck_01, loc=<4.5627 0 0>, rot=<0 0 0.0504 -0.9987>}
 head {parent=neck_02, loc=<5.3914 0 0>, rot=<0 0 0.115 -0.9934>}
 faceAttach {parent=head, loc=<2.3397 -0.1657 -0.1927>, rot=<0.0029 -0.7071 -0.0029 -0.7071>}
 C_brow_mid {parent=faceAttach, loc=<-0.181 9.5793 3.5381>, rot=<0 0 0 -1>}
 C_jaw {parent=faceAttach, loc=<0.0161 1.3026 -0.8779>, rot=<0 0 0 1>}
 C_lip_lower_mid {parent=C_jaw, loc=<-0.2183 8.0767 -3.2908>, rot=<0 0 0 -1>}
 L_lip_lower_outer {parent=C_jaw, loc=<0.7931 7.2972 -3.0902>, rot=<0.013 0.0445 0.3057 -0.951>}
 R_lip_lower_outer {parent=C_jaw, loc=<-1.2108 7.2972 -3.0902>, rot=<0.013 -0.0445 -0.3057 -0.951>}
 teeth_lower {parent=C_jaw, loc=<-0.2504 7.0213 -4.2288>, rot=<0 0 0 1>}
 tongue {parent=C_jaw, loc=<-0.0608 3.4357 -1.6853>, rot=<-0.1568 0 0 -0.9876>}
 C_lip_upper_mid {parent=faceAttach, loc=<-0.1907 9.2915 -3.81>, rot=<0.2225 0.0013 -0.0005 -0.9749>}
 C_nose_bridge {parent=faceAttach, loc=<-0.2011 10.7496 -0.8197>, rot=<-0.2218 0 0 -0.9751>}
 L_brow_mid {parent=faceAttach, loc=<2.7992 9.0525 3.7655>, rot=<-0.0001 -0.0042 0.2553 -0.9669>}
 L_brow_outer {parent=faceAttach, loc=<4.7326 7.1053 3.0633>, rot=<-0.0287 -0.04 0.4491 -0.8921>}
 L_cheek_inner {parent=faceAttach, loc=<3.0909 7.7123 -2.0261>, rot=<-0.0706 -0.2697 0.3184 -0.906>}
 L_eye {parent=faceAttach, loc=<2.4523 7.3506 2.3346>, rot=<0 0 0.9998 0.0175>}
 L_eye_lid_lower_mid {parent=L_eye, loc=<0 0 0>, rot=<0.0329 0 0 -0.9995>}
 L_eye_lid_upper_mid {parent=L_eye, loc=<0 0 0>, rot=<0.0523 0 0 -0.9986>}
 L_lip_corner {parent=faceAttach, loc=<1.8232 8.2093 -4.0021>, rot=<0.003 0.008 -0.0146 -0.9999>}
 L_lip_upper_outer {parent=faceAttach, loc=<0.8279 8.6899 -3.7566>, rot=<0.0359 -0.0728 0.2627 -0.9615>}
 R_brow_mid {parent=faceAttach, loc=<-3.1846 9.0525 3.7655>, rot=<-0.0001 0.0042 -0.2553 -0.9669>}
 R_brow_outer {parent=faceAttach, loc=<-5.0294 7.2995 3.1624>, rot=<-0.0287 0.04 -0.4491 -0.8921>}
 R_cheek_inner {parent=faceAttach, loc=<-3.4525 7.7397 -1.9958>, rot=<-0.0706 0.2697 -0.3184 -0.906>}
 R_eye {parent=faceAttach, loc=<-2.8376 7.3525 2.3347>, rot=<0 0 0.9998 -0.0175>}
 R_eye_lid_lower_mid {parent=R_eye, loc=<0 0 0>, rot=<0.0329 0 0 -0.9995>}
 R_eye_lid_upper_mid {parent=R_eye, loc=<0 0 0>, rot=<0.0523 0 0 -0.9986>}
 R_lip_corner {parent=faceAttach, loc=<-2.2086 8.2093 -4.0021>, rot=<0.003 -0.008 0.0146 -0.9999>}
 R_lip_upper_outer {parent=faceAttach, loc=<-1.2133 8.6899 -3.7566>, rot=<0.0359 0.0728 -0.2627 -0.9615>}
 teeth_upper {parent=faceAttach, loc=<-0.2093 8.5657 -2.4438>, rot=<0 0 0 1>}
 base_beak {parent=head, loc=<4.3694 -6.2492 0>, rot=<0 0 0.4617 -0.887>}
 dyn_beak {parent=base_beak, loc=<0 0 0>, rot=<0 0 0 1>}
 clavicle_r {parent=spine_05, loc=<4.3575 -0.1466 2.3223>, rot=<0.7289 0.0465 -0.6829 0.009>}
 upperarm_r {parent=clavicle_r, loc=<-15.5111 0 0>, rot=<0.0226 -0.3234 -0.0088 -0.9459>}
 upperarm_twist_01_r {parent=upperarm_r, loc=<-0.0041 0 0>, rot=<0 0 0 -1>}
]
variables [
 Neck_bend {type=double, default=0}
 Targeting_alpha {type=double, default=0}
 neck_bend_alpha {type=double, default=0}
]
